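# Physics asset for the GrandScheme backpack cosmetic, giving its mesh simple collision bodies for the pack and five upper "dyn" bones.
# The BackPack body combines a capsule and a box, while the back, front, left, right and center dyn_UP bones each carry a small capsule.
# Each dyn bone is constrained directly to BackPack in a star layout, with swing and twist limit angles set to zero.
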PhysicsAsset /BRCosmetics/Accessories/FORT_Backpacks/Backpack_M_MED_GrandScheme/Meshes/M_MED_GrandScheme_Pack_Physics
bodies {
 BackPack [
  {
   shapeType=Capsule
   length=0.274
   radius=11.3988
   shapeName=BackPack_capsule
   center=<-7.9744 -0.0032 -0.3928>
   rotation=<rot 0 90.5609 -85.5115>
  }
  {shapeType=Box, extentX=18.0307, extentY=18.0307, extentZ=18.0307, shapeName=BackPack_box, center=<4.31 0 -0.1802>, rotation=<rot 8.3913 0 0>}
 ]
 dyn_UP_Back [{shapeType=Capsule, length=0.2942, radius=1.1976, shapeName=dyn_UP_Back_capsule, rotation=<rot 0 -90 0>}]
 dyn_UP_Front [{shapeType=Capsule, length=0.2942, radius=1.1976, shapeName=dyn_UP_Front_capsule, rotation=<rot 0 -90 0>}]
 dyn_UP_Left [{
   shapeType=Capsule
   length=0.2942
   radius=1.1976
   shapeName=dyn_UP_Left_capsule
   center=<0.3038 -1.2056 0.1975>
   rotation=<rot 0.1339 -90 7.2875>
  }]
 dyn_UP_Right [{
   shapeType=Capsule
   length=0.2942
   radius=1.1976
   shapeName=dyn_UP_Right_capsule
   center=<0.2835 -2.0258 -0.202>
   rotation=<rot 0.1204 -90 -8.7871>
  }]
 dyn_UP_center [{shapeType=Capsule, length=0.2942, radius=1.1976, shapeName=dyn_UP_center_capsule, center=<1.514 0 0>, rotation=<rot 0 -90 0>}]
}
constraints [
 {bone1Name=dyn_UP_Back, bone2Name=BackPack, swing1LimitDegrees=0, swing2LimitDegrees=0, twistLimitDegrees=0}
 {bone1Name=dyn_UP_Front, bone2Name=BackPack, swing1LimitDegrees=0, swing2LimitDegrees=0, twistLimitDegrees=0}
 {bone1Name=dyn_UP_Left, bone2Name=BackPack, swing1LimitDegrees=0, swing2LimitDegrees=0, twistLimitDegrees=0}
 {bone1Name=dyn_UP_Right, bone2Name=BackPack, swing1LimitDegrees=0, swing2LimitDegrees=0, twistLimitDegrees=0}
 {bone1Name=dyn_UP_center, bone2Name=BackPack, swing1LimitDegrees=0, swing2LimitDegrees=0, twistLimitDegrees=0}
]
