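# Animation Blueprint for the GrandScheme backpack cosmetic, adding physics-driven secondary motion to the pack.
# The input pose passes through a RigidBody node using the pack's physics asset, with BackPack as base bone and partial alpha.
# Each update, the event graph reads bIsMoving2D and sets NewVar, which overrides the simulation's world gravity.
# Downward gravity is stronger when stationary and weaker during horizontal movement.
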AnimBlueprint /BRCosmetics/Accessories/FORT_Backpacks/Backpack_M_MED_GrandScheme/Meshes/M_MED_GrandScheme_Pack_AnimBP
target_skeleton=&/M_MED_GrandScheme_Pack_Skeleton
graphs [
 AnimGraph {nodes [
   n1 = RigidBody!(
    ComponentPose=Local To Component(LocalPose=Input Pose())
    Alpha=0.8
    OverrideWorldGravity:Vector=Get NewVar() -> (NewVar:Vector)
    ExternalForce
   ) -> (Pose) {
    bOverrideWorldGravity=true
    cachedBoundsScale=1
    overridePhysicsAsset=&/M_MED_GrandScheme_Pack_Physics
    alphaBoolBlend {blendOption=Linear}
    baseBoneRef {boneName=BackPack}
    componentAppliedLinearAccClamp=<1000 1000 1000>
    componentLinearAccScale {z=2}
    componentLinearVelScale {z=2}
    overrideWorldGravity {z=-90}
   }
   Output Pose(Result=Component To Local(ComponentPose=@n1)) {layerGroup=None}
  ]}
 EventGraph {nodes [
   Event BlueprintUpdateAnimation()
   > n1 = Branch(Condition=NOT Boolean(A=Get bIsMoving2D()))
   > Set NewVar(NewVar:Vector="0,0,-200") -> (Output_Get:Vector)
   Set NewVar(execute=@n1.else, NewVar="0,0,-50")
   TryGetPawnOwner()
  ]}
]
variables [NewVar {type=Vector}]
